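# Anim Blueprint for the Muster Quick SMG magazine mesh that holds a pose on the magazine bones between reloads.
# A layered blend overrides magazine and extra_magazine on the DefaultSlot pose, with alpha zero during reloads and one otherwise.
# A conduit state machine selects the short-reload pose or ref pose.
# Each update, the ShortReload curve flags short reloads, switching to the reloaded pose; the owner's long-reload montage switches it back.
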AnimBlueprint /MusterCoreWeapons/Items/Weapons/MusterQuickSMG/Animation/SMG_QuickMag_AnimBP
target_skeleton=/MusterCoreWeapons/Items/Weapons/MusterQuickSMG/Meshes/Core_S23_Quick_Mag_SMG_Skeleton
graphs [
 AnimGraph {nodes [
   AnimGraphNode_StateMachine_0 = MagazineOverride() {is=title AnimGraphNode_StateMachine, bAllowConduitEntryStates=true}
   Output Pose(Result=Layered blend per bone(
     BasePose=Use cached pose 'PostDefaultSlotPose'()
     BlendPoses_0=@AnimGraphNode_StateMachine_0
     BlendWeights_0=Get MagazineMaskAlpha() -> (MagazineMaskAlpha:double)
    ) {layerSetup [{branchFilters [{boneName=magazine}, {boneName=extra_magazine}]}]}) {layerGroup=None}
   Save cached pose 'PostDefaultSlotPose'(Pose=Slot 'DefaultSlot'(Source=Local Space Ref Pose())) {cacheName=PostDefaultSlotPose}
  ]}
 MagazineOverride {
  entry=Conduit
  path=AnimGraph.AnimGraphNode_StateMachine_0.MagazineOverride
  ids=id
  nodes [
   AnimStateConduitNode_1 = Conduit()
   AnimStateNode_0 = RefPose() {is=title AnimStateNode}
   AnimStateNode_1 = ShortReloadPose() {is=title AnimStateNode}
   AnimStateTransitionNode_2 = AnimStateTransitionNode() {crossfadeDuration=0, priorityOrder=100}
   AnimStateTransitionNode_3 = AnimStateTransitionNode() {crossfadeDuration=0}
   AnimStateTransitionNode_4 = AnimStateTransitionNode()
   AnimStateTransitionNode_5 = AnimStateTransitionNode()
  ]
  links [
   {
    dst=RefPose
    rule=AnimGraph.AnimGraphNode_StateMachine_0.MagazineOverride.AnimStateTransitionNode_2.Transition
    src=Conduit
    via=AnimStateTransitionNode_2
   }
   {
    dst=ShortReloadPose
    rule=AnimGraph.AnimGraphNode_StateMachine_0.MagazineOverride.AnimStateTransitionNode_3.Transition
    src=Conduit
    via=AnimStateTransitionNode_3
   }
   {
    dst=Conduit
    rule=AnimGraph.AnimGraphNode_StateMachine_0.MagazineOverride.AnimStateTransitionNode_5.Transition
    src=RefPose
    via=AnimStateTransitionNode_5
   }
   {
    dst=Conduit
    rule=AnimGraph.AnimGraphNode_StateMachine_0.MagazineOverride.AnimStateTransitionNode_4.Transition
    src=ShortReloadPose
    via=AnimStateTransitionNode_4
   }
  ]
 }
 Conduit {path=AnimGraph.AnimGraphNode_StateMachine_0.MagazineOverride.AnimStateConduitNode_1.Conduit, nodes [Result(bCanEnterTransition=true)]}
 RefPose {
  path=AnimGraph.AnimGraphNode_StateMachine_0.MagazineOverride.AnimStateNode_0.RefPose
  nodes [Output Animation Pose(Result=Local Space Ref Pose())]
 }
 ShortReloadPose {
  path=AnimGraph.AnimGraphNode_StateMachine_0.MagazineOverride.AnimStateNode_1.ShortReloadPose
  nodes [Output Animation Pose(Result=Sequence Player 'SMG_QuickMag_Reloaded_Pose_WEP'() {sequence=&/SMG_QuickMag_Reloaded_Pose_WEP})]
 }
 Transition {path=AnimGraph.AnimGraphNode_StateMachine_0.MagazineOverride.AnimStateTransitionNode_2.Transition, nodes [Result(bCanEnterTransition=true)]}
 Transition {
  path=AnimGraph.AnimGraphNode_StateMachine_0.MagazineOverride.AnimStateTransitionNode_3.Transition
  nodes [Result(bCanEnterTransition=Get IsShortReload() -> (IsShortReload:bool))]
 }
 Transition {
  path=AnimGraph.AnimGraphNode_StateMachine_0.MagazineOverride.AnimStateTransitionNode_4.Transition
  nodes [Result(bCanEnterTransition=Get IsLongReload() -> (IsLongReload:bool))]
 }
 Transition {
  path=AnimGraph.AnimGraphNode_StateMachine_0.MagazineOverride.AnimStateTransitionNode_5.Transition
  nodes [Result(bCanEnterTransition=Get IsShortReload())]
 }
 EventGraph {nodes [
   Event BlueprintUpdateAnimation()
   n1 = GetOwningActor()
   > Is Valid(InputObject=@n1)
   > n2 = Cast To B_SMG_MusterQuick_Athena!(execute, Object=@n1) -> (then, CastFailed, AsB SMG Muster Quick Athena:B SMG Muster Quick Athena Object Reference) {is=title K2Node_DynamicCast}
   n3 = Get Instigator!(self:Actor Object Reference=@n2) -> (Instigator:Pawn Object Reference)
   > Is Valid(InputObject=@n3)
   > n4 = Cast To FortPlayerPawn(Object=@n3)
   > Set IsShortReload(IsShortReload:bool="float > float"(A=GetCurveValue(CurveName=ShortReload))) -> (Output_Get:bool)
   > Set IsLongReload(IsLongReload:bool="Equal (Object)"(
     A=GetCurrentMontage!(self:Character Object Reference=@n4) -> (ReturnValue:Anim Montage Object Reference)
     B=&/SMG_QuickMag_Reload_Long_CMM_M
    )) -> (Output_Get:bool)
   > n5 = Branch(Condition=IsReloading!(self:Fort Weapon Object Reference=@n2) -> (ReturnValue:bool))
   > Set MagazineMaskAlpha(MagazineMaskAlpha:double=0) -> (Output_Get:double)
   Set MagazineMaskAlpha(execute=@n5.else, MagazineMaskAlpha=1)
  ]}
]
variables [
 IsLongReload {type=boolean}
 IsShortReload {type=boolean}
 MagazineMaskAlpha {type=number}
]
