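# MetaSound Source preset for the Muster Quick SMG pickup sound, a mono one-shot that inherits its graph from the shared MSS_Item_Drop source.
# It only overrides inputs, supplying a single pickup wave, a reduced master volume and a random pitch shift spread.
# Playback uses weapon reload attenuation, the weapon foley sound class and shared item concurrency.
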MetaSoundSource /MusterCoreWeapons/Items/Weapons/MusterQuickSMG/Audio/MSS_QuickSMG_Pickup_Preset
preset_parent=/FortMetasound/Items/MSS_Item_Drop
inputs [
 MasterVolume {type=Float}
 Random.PitchShift.Base {type=Float}
 Random.PitchShift.Spread {type=Float}
 Sounds {type=WaveAsset:Array}
 UE.Source.OnPlay {type=Trigger}
]
interfaces [UE.OutputFormat.Mono:1.0, UE.Source.OneShot:1.0, UE.Source:1.1]
outputs [
 "UE.OutputFormat.Mono.Audio:0" {type=Audio}
 UE.Source.OneShot.OnFinished {type=Trigger}
]
pages [Default {input_defaults {
   MasterVolume=0.2973
   Random.PitchShift.Base=0
   Random.PitchShift.Spread=0.5
   UE.Source.OnPlay=null
   Sounds [&/QuickSMG_Pickup_01]
  }}]
properties {
 assetImportData=&self:AssetImportData
 attenuationSettings=&Fort_Weapon_Reload_Att
 maxDistance=2250
 sampleRate=48000
 soundClassObject=&Fort_WeaponFoley
 concurrencySet [&CS_Fort_Item]
}
