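# Sound cue for the drop impact of the Muster Quick SMG, judging by its landed sound wave and weapon-drop attenuation.
# A single wave player feeds the root directly, with no randomization or modulation.
# It uses shared weapon drop attenuation and the loot collect sound class.
# A concurrency override caps simultaneous instances at a few and prevents new ones from starting beyond that limit.
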SoundCue /MusterCoreWeapons/Items/Weapons/MusterQuickSMG/Audio/QuickSMG_Drop_Impact_Cue
root=WavePlayer_0
nodes [{id=WavePlayer_0, properties {soundWaveAssetPtr=&/QuickSMG_Landed_01}}]
properties {
 attenuationSettings=&Fort_Weapon_Drop_Att
 bOverrideConcurrency=true
 maxDistance=3000
 soundClassObject=&Fort_Loot_Collect
 volumeMultiplier=1
 attenuationOverrides {falloffDistance=1000, attenuationShapeExtents {x=0}}
 concurrencyOverrides {maxCount=3, resolutionRule=PreventNew}
}
