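# First-person fire sound cue for the Muster Quick SMG, mixing gunshots with a low-ammo warning layer.
# The fire layer pairs a random close shot with a delayed random shell impact, with a TimeSinceStarted parameter raising pitch and easing down boosted volume.
# A LowAmmo bool branch adds a pitched-up warning shot only when true.
# Each layer overrides its sound class, and owner-limited concurrency stops the oldest instance.
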SoundCue /MusterCoreWeapons/Items/Weapons/MusterQuickSMG/Audio/QuickSMG_Fire_1P_Cue
root=Mixer_1
nodes [
 {id=Branch_3, children [Modulator_5,null,null], properties {boolParameterName=LowAmmo}}
 {id=Delay_0, children [Modulator_2], properties {delayMax=0.3, delayMin=0.2}}
 {id=Mixer_1, children [SoundClass_0,SoundClass_1], properties {inputVolume [0.7,1.25]}}
 {id=Mixer_7, children [Modulator_0,Delay_0], properties {inputVolume [1,1]}}
 {
  id=ModulatorContinuous_0
  children [Mixer_7]
  properties {
   pitchModulationParams {maxInput=0.1, minOutput=0.8, parameterName=TimeSinceStarted}
   volumeModulationParams {default=0, maxInput=0.2, minOutput=1.5, parameterName=TimeSinceStarted}
  }
 }
 {id=Modulator_0, children [Random_8]}
 {id=Modulator_2, children [Random_9], properties {pitchMax=0.95, pitchMin=0.85}}
 {id=Modulator_4, children [Branch_3], properties {pitchMax=1, pitchMin=1, volumeMax=1.15, volumeMin=1}}
 {id=Modulator_5, children [WavePlayer_26], properties {pitchMax=1.1, pitchMin=1.1, volumeMax=1.2, volumeMin=1.2}}
 {
  id=Random_8
  children [WavePlayer_36, WavePlayer_37, WavePlayer_38, WavePlayer_39, WavePlayer_40, WavePlayer_41, WavePlayer_42, WavePlayer_43]
  properties {bShouldExcludeFromBranchCulling=true, weights [1,1,1,1,1,1,1,1]}
 }
 {
  id=Random_9
  children [WavePlayer_44, WavePlayer_45, WavePlayer_46, WavePlayer_47, WavePlayer_48]
  properties {bShouldExcludeFromBranchCulling=true, weights [1,1,1,1,1]}
 }
 {
  id=SoundClass_0
  children [Modulator_4]
  properties {soundClassOverride=/Game/Sounds/FortSoundClassesAndModes/Fort_WeaponRanged_LowAmmo_Critical_PSM}
 }
 {
  id=SoundClass_1
  children [ModulatorContinuous_0]
  properties {soundClassOverride=&Fort_WeaponRanged_Fire_1P_PSM}
 }
 {id=WavePlayer_26, properties {soundWaveAssetPtr=&/QuickSMG_1P_Shot_LowAmmo_01}}
 {id=WavePlayer_36, properties {soundWaveAssetPtr=&/QuickSMG_Fire_1P_Close_OG_01}}
 {id=WavePlayer_37, properties {soundWaveAssetPtr=&/QuickSMG_Fire_1P_Close_OG_02}}
 {id=WavePlayer_38, properties {soundWaveAssetPtr=&/QuickSMG_Fire_1P_Close_OG_03}}
 {id=WavePlayer_39, properties {soundWaveAssetPtr=&/QuickSMG_Fire_1P_Close_OG_04}}
 {id=WavePlayer_40, properties {soundWaveAssetPtr=&/QuickSMG_Fire_1P_Close_OG_05}}
 {id=WavePlayer_41, properties {soundWaveAssetPtr=&/QuickSMG_Fire_1P_Close_OG_06}}
 {id=WavePlayer_42, properties {soundWaveAssetPtr=&/QuickSMG_Fire_1P_Close_OG_07}}
 {id=WavePlayer_43, properties {soundWaveAssetPtr=&/QuickSMG_Fire_1P_Close_OG_08}}
 {id=WavePlayer_44, properties {soundWaveAssetPtr=&/QuickSMG_1P_ShellImpact_01}}
 {id=WavePlayer_45, properties {soundWaveAssetPtr=&/QuickSMG_1P_ShellImpact_02}}
 {id=WavePlayer_46, properties {soundWaveAssetPtr=&/QuickSMG_1P_ShellImpact_03}}
 {id=WavePlayer_47, properties {soundWaveAssetPtr=&/QuickSMG_1P_ShellImpact_04}}
 {id=WavePlayer_48, properties {soundWaveAssetPtr=&/QuickSMG_1P_ShellImpact_05}}
]
properties {
 attenuationSettings=&Fort_Weapon_Projectile_FirstPerson_Athena_Att
 bDebug=true
 maxDistance=2097152
 soundClassObject=&Fort_WeaponRanged_Fire_1P_PSM
 volumeMultiplier=1
 attenuationOverrides {manualReverbSendLevel=-1.0165, reverbSendMethod=Manual}
 concurrencyOverrides {bLimitToOwner=true, maxCount=2, resolutionRule=StopOldest, volumeScale=0.9}
 concurrencySet [&Weapon_Gunfire_FirstPerson_Conc]
}
