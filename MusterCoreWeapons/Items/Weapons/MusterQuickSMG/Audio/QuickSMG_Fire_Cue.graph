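# Quick SMG fire sound cue that layers close, medium and distant gunshot variations, with a root modulator varying pitch by the WeaponID parameter.
# Each layer randomizes its waves through a modulator, its own attenuation and a sound class override before nested mixers blend them.
# Only the close layer, built from 1P-named waves under a 3P class, has pitch and volume shaped by TimeSinceStarted.
# Per-owner concurrency caps instances, stopping the oldest.
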SoundCue /MusterCoreWeapons/Items/Weapons/MusterQuickSMG/Audio/QuickSMG_Fire_Cue
root=ModulatorContinuous_2
nodes [
 {id=Attenuation_3, children [Modulator_3], properties {attenuationSettings=&Fort_Weapon_Projectile_Close_Athena_Att}}
 {
  id=Attenuation_4
  children [Modulator_4]
  properties {attenuationSettings=&Fort_Weapon_Projectile_Distant_Mono_Athena_Att}
 }
 {
  id=Attenuation_5
  children [Modulator_1]
  properties {attenuationSettings=&Fort_Weapon_Projectile_Medium_Mono_Athena_Att}
 }
 {id=Mixer_2, children [Mixer_3,SoundClass_4], properties {inputVolume [1,0.6]}}
 {id=Mixer_3, children [ModulatorContinuous_0, SoundClass_5], properties {inputVolume [0.7,1]}}
 {
  id=ModulatorContinuous_0
  children [SoundClass_3]
  properties {
   pitchModulationParams {maxInput=0.1, minOutput=0.8, parameterName=TimeSinceStarted}
   volumeModulationParams {default=0, maxInput=0.2, minOutput=1.5, parameterName=TimeSinceStarted}
  }
 }
 {
  id=ModulatorContinuous_2
  children [Mixer_2]
  properties {pitchModulationParams {default=0, maxInput=4, maxOutput=1.4, minOutput=1.05, parameterName=WeaponID}}
 }
 {id=Modulator_1, children [Random_2], properties {volumeMax=1, volumeMin=0.9}}
 {id=Modulator_3, children [Random_0], properties {pitchMax=1.1, pitchMin=1, volumeMax=0.95, volumeMin=1.05}}
 {id=Modulator_4, children [Random_3]}
 {
  id=Random_0
  children [WavePlayer_0,WavePlayer_1,WavePlayer_2,WavePlayer_3,WavePlayer_4,WavePlayer_5,WavePlayer_6,WavePlayer_7]
  properties {bShouldExcludeFromBranchCulling=true, weights [1,1,1,1,1,1,1,1]}
 }
 {
  id=Random_2
  children [WavePlayer_17, WavePlayer_18, WavePlayer_19, WavePlayer_20, WavePlayer_21, WavePlayer_22]
  properties {bShouldExcludeFromBranchCulling=true, weights [1,1,1,1,1,1]}
 }
 {
  id=Random_3
  children [WavePlayer_11, WavePlayer_12, WavePlayer_13, WavePlayer_14, WavePlayer_15, WavePlayer_16]
  properties {bShouldExcludeFromBranchCulling=true, weights [1,1,1,1,1,1]}
 }
 {
  id=SoundClass_3
  children [Attenuation_3]
  properties {soundClassOverride=&Fort_WeaponRanged_Fire_3P_PSM}
 }
 {
  id=SoundClass_4
  children [Attenuation_4]
  properties {soundClassOverride=&Fort_WeaponRanged_Fire_Distant}
 }
 {
  id=SoundClass_5
  children [Attenuation_5]
  properties {soundClassOverride=&Fort_WeaponRanged_Fire_Distant}
 }
 {id=WavePlayer_0, properties {soundWaveAssetPtr=&/QuickSMG_Fire_1P_Close_OG_01}}
 {id=WavePlayer_1, properties {soundWaveAssetPtr=&/QuickSMG_Fire_1P_Close_OG_02}}
 {id=WavePlayer_11, properties {soundWaveAssetPtr=&/QuickSMG_Fire_3P_Distant_01}}
 {id=WavePlayer_12, properties {soundWaveAssetPtr=&/QuickSMG_Fire_3P_Distant_02}}
 {id=WavePlayer_13, properties {soundWaveAssetPtr=&/QuickSMG_Fire_3P_Distant_03}}
 {id=WavePlayer_14, properties {soundWaveAssetPtr=&/QuickSMG_Fire_3P_Distant_04}}
 {id=WavePlayer_15, properties {soundWaveAssetPtr=&/QuickSMG_Fire_3P_Distant_05}}
 {id=WavePlayer_16, properties {soundWaveAssetPtr=&/QuickSMG_Fire_3P_Distant_06}}
 {id=WavePlayer_17, properties {soundWaveAssetPtr=&/QuickSMG_Fire_3P_Medium_01}}
 {id=WavePlayer_18, properties {soundWaveAssetPtr=&/QuickSMG_Fire_3P_Medium_02}}
 {id=WavePlayer_19, properties {soundWaveAssetPtr=&/QuickSMG_Fire_3P_Medium_03}}
 {id=WavePlayer_2, properties {soundWaveAssetPtr=&/QuickSMG_Fire_1P_Close_OG_03}}
 {id=WavePlayer_20, properties {soundWaveAssetPtr=&/QuickSMG_Fire_3P_Medium_04}}
 {id=WavePlayer_21, properties {soundWaveAssetPtr=&/QuickSMG_Fire_3P_Medium_05}}
 {id=WavePlayer_22, properties {soundWaveAssetPtr=&/QuickSMG_Fire_3P_Medium_06}}
 {id=WavePlayer_3, properties {soundWaveAssetPtr=&/QuickSMG_Fire_1P_Close_OG_04}}
 {id=WavePlayer_4, properties {soundWaveAssetPtr=&/QuickSMG_Fire_1P_Close_OG_05}}
 {id=WavePlayer_5, properties {soundWaveAssetPtr=&/QuickSMG_Fire_1P_Close_OG_06}}
 {id=WavePlayer_6, properties {soundWaveAssetPtr=&/QuickSMG_Fire_1P_Close_OG_07}}
 {id=WavePlayer_7, properties {soundWaveAssetPtr=&/QuickSMG_Fire_1P_Close_OG_08}}
]
properties {
 bDebug=true
 bOverrideConcurrency=true
 maxDistance=27500
 soundClassObject=&Fort_WeaponRanged_Fire_3P_PSM
 concurrencyOverrides {bLimitToOwner=true, maxCount=2, resolutionRule=StopOldest}
}
