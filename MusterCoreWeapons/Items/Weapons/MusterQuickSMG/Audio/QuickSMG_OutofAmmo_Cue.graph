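# Sound cue for the Quick SMG out-of-ammo sound, routed through the third-person ranged weapon fire sound class.
# A Random node picks between two equally weighted wave variations, and a Modulator above it slightly varies volume while keeping pitch fixed.
# The cue lowers overall volume, sets an attenuation falloff distance, and overrides concurrency to allow few simultaneous instances, stopping the oldest.
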SoundCue /MusterCoreWeapons/Items/Weapons/MusterQuickSMG/Audio/QuickSMG_OutofAmmo_Cue
root=Modulator_0
nodes [
 {id=Modulator_0, children [Random_0], properties {pitchMax=1, pitchMin=1, volumeMax=0.95, volumeMin=0.85}}
 {id=Random_0, children [WavePlayer_0,WavePlayer_1], properties {weights [1,1]}}
 {id=WavePlayer_0, properties {soundWaveAssetPtr=&/QuickSMG_OutofAmmo_01}}
 {id=WavePlayer_1, properties {soundWaveAssetPtr=&/QuickSMG_OutofAmmo_02}}
]
properties {
 bOverrideConcurrency=true
 maxDistance=2097152
 soundClassObject=&Fort_WeaponRanged_Fire_3P_PSM
 volumeMultiplier=0.55
 attenuationOverrides {falloffDistance=1600}
 concurrencyOverrides {maxCount=2, resolutionRule=StopOldest}
}
