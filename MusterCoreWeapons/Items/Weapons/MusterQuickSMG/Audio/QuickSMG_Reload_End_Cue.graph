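# Sound cue for the end of the Muster Quick SMG reload, built from first-person foley waves.
# A Random node picks one of three equally weighted wave variations each time it plays.
# It uses the Fort weapon reload attenuation and weapon foley sound class at a reduced volume.
# Concurrency is overridden on the cue itself, and its debug flag is left on.
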SoundCue /MusterCoreWeapons/Items/Weapons/MusterQuickSMG/Audio/QuickSMG_Reload_End_Cue
root=Random_0
nodes [
 {id=Random_0, children [WavePlayer_1,WavePlayer_2,WavePlayer_3], properties {weights [1,1,1]}}
 {id=WavePlayer_1, properties {soundWaveAssetPtr=&/QuickSMG_1P_Reload_End_01}}
 {id=WavePlayer_2, properties {soundWaveAssetPtr=&/QuickSMG_1P_Reload_End_02}}
 {id=WavePlayer_3, properties {soundWaveAssetPtr=&/QuickSMG_1P_Reload_End_03}}
]
properties {
 attenuationSettings=&Fort_Weapon_Reload_Att
 bDebug=true
 bOverrideConcurrency=true
 maxDistance=2250
 soundClassObject=&Fort_WeaponFoley
 volumeMultiplier=0.752
}
