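# Sound cue for the QuickSMG fast reload insert foley in MusterCoreWeapons.
# A Random node picks with equal weight between two fast cartridge wave players.
# It plays in the weapon foley sound class with shared weapon reload attenuation, slightly reduced volume and a concurrency override.
# The debug flag is left enabled.
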SoundCue /MusterCoreWeapons/Items/Weapons/MusterQuickSMG/Audio/QuickSMG_Reload_InsertFast_Cue
root=Random_1
nodes [
 {id=Random_1, children [WavePlayer_0,WavePlayer_4], properties {weights [1,1]}}
 {id=WavePlayer_0, properties {soundWaveAssetPtr=&/QuickSMG_Reload_Cartridge_Fast_01}}
 {id=WavePlayer_4, properties {soundWaveAssetPtr=&/QuickSMG_Reload_Cartridge_Fast_02}}
]
properties {
 attenuationSettings=&Fort_Weapon_Reload_Att
 bDebug=true
 bOverrideConcurrency=true
 maxDistance=2250
 soundClassObject=&Fort_WeaponFoley
 volumeMultiplier=0.852
}
