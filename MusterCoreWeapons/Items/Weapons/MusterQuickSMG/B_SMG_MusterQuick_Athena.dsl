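; Weapon Blueprint for the MusterQuick SMG (Athena) that tags its fire audio with a per-instance ID.
; On BeginPlay, after the parent call, it has its BPC_Audio_ID component assign a random weapon ID.
; On each weapon fire FX event, it calls the parent, then passes that ID as the float "WeaponID" audio parameter.
; This differentiates third-person fire sounds when many instances of the same weapon fire at once.
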(event EventBeginPlay
  (|CallParentFunction)
  ;[0 0.3094 0.04008] Assign Random Weapon ID used to differentiate many instances of 3P Fire in-game
  (Class|BPCAudioID|AssignWeaponID (Variables|Default|GetBPC_Audio_Weapon_ID)))

(event EventOnPlayWeaponFireFX
  (params (bPersistentFire :type Boolean) (bSecondaryFire :type Boolean))
  ;[0.1455 0.3094 0] "WeaponID" is used to help differentiate 3P Fire when there are many instances of the same weapon in game firing at once.
  (|CallParentFunction bPersistentFire bSecondaryFire)
  (Audio|Parameter|SetFloatParameter (AudioParameter|GetAudioParameterComponent) "WeaponID" (Math|Conversions|ToFloat(Integer) (Variables|/Game/Sounds/Fort_Audio_Blueprints/AudioIDComponent/BPC_Audio_ID.BPC_Audio_ID_C|GetWeaponID (Variables|Default|GetBPC_Audio_Weapon_ID))) :!class /Script/AudioExtensions.AudioParameterControllerInterface))

(fn ConstructionScript
  (|CallParentFunction))
